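; Destructible prototype of the spinning lighthouse light that shuts the lighthouse effect and audio off when destroyed.
; The server sets Is Died at match ready and on death; its OnRep restores the LighthouseControl collection's Alpha or calls KillLighthouse.
; KillLighthouse zeroes Alpha, hides the Fresnel lens and flags the first coastal lighthouse audio actor dead, also running if destroyed while alive.
; Remote clients play the spinning audio, stopped by death effects.
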(event EventBeginPlay
  (|CallParentFunction)
  (Networking|SwitchHasAuthority
    (:Authority
      (Class|FortAsyncActionAuthorityMatchReady|MatchReadyAsync
        (:OnReady
          (Variables|Default|SetIs%20Died 0))))
    (:Remote
      (Audio|Components|Audio|Play (Variables|Default|GetSpinningLightAudio) 0.0))))

(event EventOnDeathServer
  (params (Damage :type Float%20%28single-precision%29) (DamageTags :type Gameplay%20Tag%20Container%20Structure%20%28by%20ref%29) (Momentum :type Vector :default "0, 0, 0") (HitInfo :type Hit%20Result%20Structure%20%28by%20ref%29) (InstigatedBy :type Controller%20Object%20Reference) (DamageCauser :type Actor%20Object%20Reference) (EffectContext :type Gameplay%20Effect%20Context%20Handle%20Structure))
  (|CallParentFunction Damage DamageTags Momentum HitInfo InstigatedBy DamageCauser EffectContext)
  (Variables|Default|SetIs%20Died 1))

(event EventOnDeathPlayEffects
  (params (Damage :type Float%20%28single-precision%29) (DamageTags :type Gameplay%20Tag%20Container%20Structure%20%28by%20ref%29) (Momentum :type Vector :default "0, 0, 0") (HitInfo :type Hit%20Result%20Structure%20%28by%20ref%29) (InstigatedBy :type Fort%20Pawn%20Object%20Reference) (DamageCauser :type Actor%20Object%20Reference) (EffectContext :type Gameplay%20Effect%20Context%20Handle%20Structure))
  (Audio|Components|Audio|Stop (Variables|Default|GetSpinningLightAudio)))

(event EventDestroyed
  (if (== (Variables|Default|GetIs%20Died) 0)
    (CallFunction|KillLighthouse)))

(fn killLighthouse
  (Rendering|Material|SetScalarParameterValue "/Game/Effects/Fort_Effects/World/Enviroment/Lighthouse/LighthouseControl.LighthouseControl" "Alpha" :!class /Script/Engine.KismetMaterialLibrary)
  (Development|SetHiddeninGame (Variables|Default|GetFresnelLens) true true)
  (bind _outactors (Actor|GetAllActorsOfClass "/Game/Sounds/Apollo/Biomes/Coastal_Areas/BP_Biome_Coastal_Lighthouse_Audio.BP_Biome_Coastal_Lighthouse_Audio_C"))
  (Utilities|IsValid (Utilities|Array|Get(acopy) _outactors)
    (:Is%20Valid
      (Variables|/Game/Sounds/Apollo/Biomes/Coastal_Areas/BP_Biome_Coastal_Lighthouse_Audio.BP_Biome_Coastal_Lighthouse_Audio_C|SetbIsLighthouseDead true (Utilities|Array|Get(acopy) _outactors)))))

(fn OnRep_IsDied
  (switch int (Variables|Default|GetIs%20Died)
    (:0
      (Rendering|Material|SetScalarParameterValue "/Game/Effects/Fort_Effects/World/Enviroment/Lighthouse/LighthouseControl.LighthouseControl" "Alpha" 1.0 :!class /Script/Engine.KismetMaterialLibrary))
    (:1
      (CallFunction|KillLighthouse))))

(fn ConstructionScript)
