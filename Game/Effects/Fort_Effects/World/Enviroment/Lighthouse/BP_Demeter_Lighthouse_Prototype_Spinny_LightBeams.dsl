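; Lighthouse light-beam prototype Blueprint that sets up two fresnel cone meshes in construction and at BeginPlay, where cosmetic events run.
; Cone length scales with material quality level; XboxOne instead scales beam A1 at quarter length, while other platforms scale A2.
; Materials come from override variables or Android mobile defaults.
; Cone angle, start radius, color and light shift derive shader parameters, set on A2 except Zmin on A1.
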(event EventBeginPlay
  (|CallParentFunction)
  (CallFunction|InitBeam))

(fn getValidatedConeLength
  (return (* (Variables|Default|GetCone%20Length) (Utilities|Select :Index (Rendering|Material|GetRenderingMaterialQualityLevel) :Option%200 0.25 :Option%201 0.5 :Option%202 0.5 :Option%203 1.0))))

(fn initBeam
  (Networking|CanExecuteCosmeticEvents
    (:True
      (if (Utilities|String|Equal,CaseInsensitive(String) (Game|GetPlatformName) "XboxOne")
        (Transformation|SetWorldScale3D (Variables|Default|GetFresnelBeam_A1) (Math|Vector|MakeVector 0.01 0.01 (/ (* (Variables|Default|GetCone%20Length) 0.25) 1e+02)))
        (else
          (Transformation|SetWorldScale3D (Variables|Default|GetFresnelBeam_A2) (Math|Vector|MakeVector 0.01 0.01 (/ (CallFunction|GetValidatedConeLength) 1e+02)))))
      (if (Variables|Default|GetOverride_Beam_Materials)
        (CallFunction|OverrideBeamMaterials)
        (elif (Utilities|String|Equal,CaseInsensitive(String) (Game|GetPlatformName) "Android")
          (Rendering|Material|SetMaterial (Variables|Default|GetFresnelBeam_A1) 0 "/Game/Effects/Fort_Effects/World/Enviroment/Volumetrics/LightBeams/MI_LightCone_Lighthouse_Flipped_Mobile.MI_LightCone_Lighthouse_Flipped_Mobile")
          (Rendering|Material|SetMaterial (Variables|Default|GetFresnelBeam_A2) 0 "/Game/Effects/Fort_Effects/World/Enviroment/Volumetrics/LightBeams/MI_LightCone_Lighthouse_Mobile.MI_LightCone_Lighthouse_Mobile")))
      (CallFunction|InitializeMaterialParams))))

(fn initializeMaterialParams
  (bind _light_color_rgb__and_brightness_a (Variables|Default|GetLight%20Color%28rgb%29%20and%20Brightness%28a%29))
  (bind (_r _g _b _a) (Math|Color|BreakColor _light_color_rgb__and_brightness_a))
  (bind _light_shift (Variables|Default|GetLight%20Shift))
  (bind _returnvalue (Math|Float|Max(Float) _light_shift 1))
  (bind _returnvalue_2 (/ 1 _returnvalue))
  (bind _returnvalue_3 (/ _a _returnvalue_2))
  (bind _cone_angle (Variables|Default|GetCone%20Angle))
  (bind _returnvalue_4 (Math|Trig|Tan(Degrees) _cone_angle))
  (bind _returnvalue_5 (/ 1 _returnvalue_4))
  (bind _start_radius (Variables|Default|GetStart%20Radius))
  (bind _returnvalue_6 (* _start_radius _returnvalue_5))
  (bind _fresnelbeam_a2 (Variables|Default|GetFresnelBeam_A2))
  (Rendering|Material|SetScalarParameterValueonMaterials _fresnelbeam_a2 "R1" (Variables|Default|GetStart%20Radius))
  (Rendering|Material|SetScalarParameterValueonMaterials _fresnelbeam_a2 "R2" (* (Math|Trig|Tan(Degrees) (Variables|Default|GetCone%20Angle)) (+ (CallFunction|GetValidatedConeLength) _returnvalue_6)))
  (Rendering|Material|SetScalarParameterValueonMaterials (Variables|Default|GetFresnelBeam_A1) "Zmin" _returnvalue_6)
  (Rendering|Material|SetScalarParameterValueonMaterials _fresnelbeam_a2 "ZMax" (+ (CallFunction|GetValidatedConeLength) _returnvalue_6))
  (Rendering|Material|SetScalarParameterValueonMaterials _fresnelbeam_a2 "Tan" (Math|Trig|Tan(Degrees) (Variables|Default|GetCone%20Angle)))
  (Rendering|Material|SetVectorParameterValueonMaterials _fresnelbeam_a2 "Color" (Math|Conversions|ToVector(LinearColor) (Math|Color|MakeColor _r _g _b _returnvalue_3)))
  (Rendering|Material|SetScalarParameterValueonMaterials _fresnelbeam_a2 "EmissiveIntensity" _returnvalue_3)
  (Rendering|Material|SetScalarParameterValueonMaterials _fresnelbeam_a2 "LightShift" (Variables|Default|GetLight%20Shift)))

(fn OverrideBeamMaterials
  (if (Utilities|String|Equal,CaseInsensitive(String) (Game|GetPlatformName) "Android")
    (Rendering|Material|SetMaterial (Variables|Default|GetFresnelBeam_A1) 0 (Variables|Default|GetBeam1_Material_Mobile))
    (Rendering|Material|SetMaterial (Variables|Default|GetFresnelBeam_A2) 0 (Variables|Default|GetBeam2_Material_Mobile))
    (else
      (Rendering|Material|SetMaterial (Variables|Default|GetFresnelBeam_A1) 0 (Variables|Default|GetBeam1_Material))
      (Rendering|Material|SetMaterial (Variables|Default|GetFresnelBeam_A2) 0 (Variables|Default|GetBeam2_Material)))))

(fn ConstructionScript
  (CallFunction|InitBeam))
