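# Editor-only unlit material for the lighthouse environment effect, named as a "light cooker"; its exact role is not shown.
# A warm pale-yellow Color vector parameter has its RGB multiplied by itself, squaring each channel to deepen the tint; its alpha is unused.
# The graph shows no connection from that product to a material output.
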Material /Game/Effects/Fort_Effects/World/Enviroment/Lighthouse/EditorOnly_LighthouseLightCooker_Mat
graph {
 properties {pixelDepthOffsetMode=PDOM_Legacy, shadingModel=MSM_Unlit, usedShadingModels=Unlit}
 nodes [
  n1 = VectorParameter() {parameterName=Color, defaultValue=<rgba 0.9583 0.7971 0.4211 0.1>}
  Multiply(A=@n1.RGB, B=@n1.RGB)
 ]
}
