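# Material function that spins lighthouse geometry continuously around an axis parameter, with speed shared through the LighthouseControl parameter collection.
# Time multiplied by Rotation Speed gives the angle that rotates vertices about the object pivot, added to incoming WPO.
# The same axis and angle feed a vertex-interpolated tangent-space normal fix and a rotated world-space vertex normal.
# A further output rotates an input vector, defaulting to the transformed local Y axis.
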MaterialFunction /Game/Effects/Fort_Effects/World/Enviroment/Lighthouse/LighthouseRotationandControl
graph {nodes [
  n1 = VectorParameter() {group=Lighthouse Control, parameterName=Rotation Axis, defaultValue=<rgba 0 0 1 0>}
  n2 = Multiply(
   A=Time()
   B=CollectionParameter() {collection=&/LighthouseControl, parameterName=Rotation Speed}
  )
  n3 = MaterialFunctionCall() {materialFunction=&ObjectPivotPoint}
  n4 = WorldPosition()
  FunctionOutput(None=Add(
    A=FunctionInput() {bUsePreviewValueAsDefault=true, inputName=WPO}
    B=RotateAboutAxis(NormalizedRotationAxis=@n1.RGB, RotationAngle=@n2, PivotPoint=@n3.Object Pivot Location, Position=@n4.XYZ)
   )) {outputName=WPO}
  FunctionOutput(None=VertexInterpolator(VS=MaterialFunctionCall(Rotation Axis=@n1.RGB, Rotation Angle=@n2) {materialFunction=&FixRotateAboutAxisNormals}.New Tangent Space Vertex Normal).PS) {outputName=Normal, sortPriority=1}
  n5 = Constant()
  n6 = VertexNormalWS()
  FunctionOutput(None=Add(A=RotateAboutAxis(NormalizedRotationAxis=@n1.RGB, RotationAngle=@n2, PivotPoint=@n5, Position=@n6), B=@n6)) {outputName="Vertex Normal (Corrected)", sortPriority=1}
  n7 = FunctionInput(Preview=Transform(None=Constant3Vector() {constant=<rgba 0 1 0 0>}) {transformSourceType=TRANSFORMSOURCE_Local}) {bUsePreviewValueAsDefault=true, inputName=Postion to Transform}
  FunctionOutput(None=Add(A=RotateAboutAxis(NormalizedRotationAxis=@n1.RGB, RotationAngle=@n2, PivotPoint=@n5, Position=@n7), B=@n7)) {outputName=Rotated Position, sortPriority=2}
 ]}
