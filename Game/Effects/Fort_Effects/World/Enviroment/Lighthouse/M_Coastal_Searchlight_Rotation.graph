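# Material for the coastal lighthouse searchlight, rotating it through World Position Offset and lighting it by time of day.
# LighthouseRotationandControl takes hit-and-shrink damage offsets and outputs the final offset plus a normal, blended with the normal map.
# Tinted diffuse color drives emissive, masked by SpecularMasks alpha and scaled brighter at night, while that texture's red channel feeds specular, metallic and roughness.
# Attributes then pass through Creative move-tool and emissive-ramp functions.
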Material /Game/Effects/Fort_Effects/World/Enviroment/Lighthouse/M_Coastal_Searchlight_Rotation
graph {
 properties {
  bUseMaterialAttributes=true
  bUsedWithInstancedStaticMeshes=true
  bUsedWithNanite=true
  ditheredLODTransition=true
  pixelDepthOffsetMode=PDOM_Legacy
  usedShadingModels=Default Lit
 }
 nodes [
  n1 = TextureCoordinate()
  n2 = TextureObjectParameter() {
   group=Material_1
   parameterName=SpecularMasks
   samplerType=SAMPLERTYPE_Masks
   sortPriority=2
   texture=&T_Fortnite_Default_S
  }
  n3 = TextureSample(UVs=@n1, Tex=@n2) {samplerType=SAMPLERTYPE_Masks}
  n4 = TextureSample(
   UVs=@n1
   Tex=TextureObjectParameter() {group=Material_1, parameterName=Diffuse, sortPriority=1, texture=/Game/Global/Textures/Default/Blanks/T_MidgreysRGB}
  )
  n5 = Multiply(
   A=ComponentMask(None=@n4.RGBA) {b=true, g=true, r=true}
   B=VectorParameter() {parameterName=Emissive Color, defaultValue=<rgba 4 3.5635 1.8581 1>}.RGB
  )
  n6 = ScalarParameter() {defaultValue=1, group=Emissive, parameterName=DayMult, sortPriority=1}
  n7 = MaterialFunctionCall(
   Color_Morning=@n6
   Color_Day=@n6
   Color_Evening=@n6
   Color_Night=ScalarParameter() {defaultValue=5, group=Emissive, parameterName=NightMult, sortPriority=2}
  ) {materialFunction=&Color_Time_of_Day}
  n8 = MaterialFunctionCall!(
   Postion to Transform
   WPO=MaterialFunctionCall!(
    WSDeformationCenter=VectorParameter() {bUseCustomPrimitiveData=true, group=z_code, parameterName=WSDeformationCenter, primitiveDataIndex=20}.RGB
    Code_DestructionAmount=ScalarParameter() {group=z_code, parameterName=Code_DestructionAmount}
    Code_UseUvForDamage
    Code_UseHitEffects=StaticBoolParameter() {defaultValue=true, group=z_code, parameterName=Code_UseHitEffects}
    Code_Emissive
    WSDamageRadius=ScalarParameter() {bUseCustomPrimitiveData=true, defaultValue=1, group=z_code, parameterName=WSDamageRadius, primitiveDataIndex=26}
    CrackDamageShader
    Diffuse
    ScalarModDamageCracks
    ScalarOffsetforDamageCracks
    ExposeLerpFunctionForInvertingTheHitDistanceCalc
    InvertDamageFalloff
    CrackWorldPosition
    WorldPosition
    DistanceCullFadeAnimationOffset=ScalarParameter() {bUseCustomPrimitiveData=true, group=z_code, parameterName=DistanceCullFadeAnimationOffset, primitiveDataIndex=12}
   ) {materialFunction=/Game/Packages/jl_tests/Fortnite_HitEffects}.Hit And Shrink WPO
  ) {materialFunction=&/LighthouseRotationandControl}
  n9 = TextureObjectParameter() {group=Material_1, parameterName=Normals, samplerType=SAMPLERTYPE_Normal, sortPriority=3, texture=/Engine/EngineMaterials/FlatNormal}
  n10 = SetMaterialAttributes!(
   MaterialAttributes
   Base Color=@n4.RGBA
   Specular=@n3.R
   Metallic=@n3.R
   Roughness=@n3.R
   Emissive Color=LinearInterpolate(B=Multiply(A=@n5, B=@n7.Result), Alpha=@n3.A)
   Normal=MaterialFunctionCall(BaseNormal=@n8.Normal, AdditionalNormal=TextureSample(UVs=@n1, Tex=@n9) {samplerType=SAMPLERTYPE_Normal}.RGB) {materialFunction=&BlendAngleCorrectedNormals}.Result
   Custom Data 0=ComponentMask(None=@n4.RGBA) {a=true}
   Custom Data 1=Constant() {r=0.025}
   World Position Offset=@n8.WPO
   Customized UV 1
   Customized UV 2
  ) {attributeSetTypes [
    #1
    #2
    #3
    #4
    #5
    #6
    #7
    #8
    #9
    #10
    #11
   ]}
  MaterialFunctionCall!(MA In=MaterialFunctionCall(MA In=@n10) {materialFunction=&MF_CreativeMoveTool}.MA Out) {materialFunction=/Game/Weapons/Prototype/ObjectMover/Materials/MF_Creative_EmissiveRamp}
 ]
}
